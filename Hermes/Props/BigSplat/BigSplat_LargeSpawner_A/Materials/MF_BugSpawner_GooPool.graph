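# Material function that adds glowing, animated goo to the BigSplat bug spawner's material, overriding base color, emissive and World Position Offset.
# Goo color combines inner/outer tints, Fresnel, distorted world-aligned caustics, an AO/thickness texture, time-of-day scaling and an input-driven pulse.
# Panned and world-aligned noise drive dripping and normal-based inflation offsets, with vertex color red supplying masks.
# Custom nodes read the directional light, falling back to a collection sun vector.
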MaterialFunction /Hermes/Props/BigSplat/BigSplat_LargeSpawner_A/Materials/MF_BugSpawner_GooPool
graph {nodes [
  n1 = GetMaterialAttributes(None=FunctionInput() {inputName=Material, inputType=FunctionInput_MaterialAttributes}) {attributeGetTypes [#1, #2, #3]}
  FunctionOutput(None=SetMaterialAttributes!(
    MaterialAttributes=@n1.MaterialAttributes
    WorldPositionOffset=NamedRerouteUsage()
    EmissiveColor=NamedRerouteUsage()
    BaseColor=NamedRerouteUsage()
   ) {attributeSetTypes [#1, #2, #4]})
  NamedRerouteDeclaration(None=@n1.WorldPositionOffset) {name=WPO_IN, nodeColor=<rgba 1 0.1172 0 1>}
  n2 = NamedRerouteUsage()
  n3 = Add(A=NamedRerouteUsage(), B=NamedRerouteUsage()) {constB=0}
  NamedRerouteDeclaration(None=Add(A=NamedRerouteUsage(), B=Multiply(A=@n2, B=@n3))) {name=WPO_OUT, nodeColor=<rgba 0 0.2031 1 1>}
  n4 = NamedRerouteUsage()
  n5 = ScalarParameter() {defaultValue=1, group=Goo, parameterName=ThicknessMulti}
  NamedRerouteDeclaration(None=LinearInterpolate(A=Multiply(A=@n4, B=@n5), B=@n4, Alpha=NamedRerouteUsage())) {name=Emissive_IN, nodeColor=<rgba 0.0078 1 0 1>}
  n6 = Custom() {code="return View.DirectionalLightDirection;", description=DirectionalLightVector, inputs []}
  n7 = DotProduct(A=@n6, B=@n6)
  n8 = CollectionParameter() {
   collection=&FortniteMaterialParameters
   parameterName=SunlightDirectionalVector
  }
  n9 = ComponentMask(None=@n8) {b=true, g=true, r=true}
  NamedRerouteDeclaration(None=If(
    A=Custom() {code="return View.DirectionalLightColor;", description=DirectionalLightColor, inputs []}
    "A > B"=If(A=@n7, "A > B"=@n6, "A == B"=@n9, "A < B"=@n9) {desc=check if DirectionalLight exists}
    "A == B"=@n9
    "A < B"=@n9
   ) {desc=check if DirectionalLight color is Black}) {name=SunDir, nodeColor=<rgba 0 1 0.0391 1>}
  n10 = Constant() {r=0.1}
  NamedRerouteDeclaration(None=OneMinus(None=Fresnel(ExponentIn=@n10) {baseReflectFraction=0, exponent=0.1})) {name=InvertedFresnel, nodeColor=<rgba 0.125 1 0 1>}
  n11 = NamedRerouteUsage()
  n12 = LinearInterpolate(
   A=@n11
   B=Multiply(A=@n11, B=ScalarParameter() {defaultValue=6, group=Goo, parameterName=GooCenterBrightness}) {constB=1.5}
   Alpha=NamedRerouteUsage()
  )
  n13 = Saturate(None=DotProduct(
    A=NamedRerouteUsage()
    B=MaterialFunctionCall() {materialFunction=&CameraDirectionVector}.Result
   ))
  n14 = NamedRerouteUsage()
  n15 = NamedRerouteUsage()
  n16 = Multiply(A=Saturate(None=Power(Base=@n15) {constExponent=2.5}), B=ScalarParameter() {defaultValue=0.2, group=Goo, parameterName=PulsingIntensity})
  n17 = Add(
   A=LinearInterpolate(A=@n11, B=LinearInterpolate(A=@n11, B=@n12, Alpha=@n13), Alpha=NamedRerouteUsage())
   B=Multiply(A=Multiply(A=@n14, B=@n16), B=VectorParameter() {group=Goo, parameterName=Goo_PulsingTint, defaultValue=<rgba 1.4808 0.4922 1.5 1>}.RGB)
  )
  n18 = NamedRerouteUsage()
  n19 = Multiply(A=@n18, B=ScalarParameter() {defaultValue=1, parameterName=EmissiveLerp})
  NamedRerouteDeclaration(None=LinearInterpolate(B=@n17, Alpha=Saturate(None=@n19)) {constB=0}) {name=Emissive_OUT, nodeColor=<rgba 0.4062 0 1 1>}
  NamedRerouteDeclaration(None=FunctionInput() {inputName=Pulsing, inputType=FunctionInput_Scalar, sortPriority=1}) {name=Pulsing, nodeColor=<rgba 0 0.25 1 1>}
  n20 = VertexColor()
  NamedRerouteDeclaration(None=OneMinus(None=@n20.R)) {name=PulsingMask, nodeColor=<rgba 1 0 0.0469 1>}
  n21 = PreSkinnedNormal()
  NamedRerouteDeclaration(None=ShaderStageSwitch(PixelShader=VertexInterpolator(VS=@n21).PS, VertexShader=@n21)) {name=PreSkinnedNrm_VS-PS, nodeColor=<rgba 0 0.1943 0.401 1>}
  n22 = LocalPosition() {localOrigin=InstancePreSkinning}
  NamedRerouteDeclaration(None=ShaderStageSwitch(PixelShader=VertexInterpolator(VS=@n22.XYZ).PS, VertexShader=@n22.XYZ)) {name=PreSkinnedPos_VS-PS, nodeColor=<rgba 0 0.4844 1 1>}
  n23 = TextureSampleParameter2D() {
   group=Goo
   parameterName=Goo_AOThickness
   texture=/Hermes/Props/BigSplat/BigSplat_LargeSpawnerDestroyed_A/Textures/Nanite/T_NITE_BigSplat_LargeSpawnerDestroyedGoo_A_A_AT
  }
  NamedRerouteDeclaration(None=Power(Base=@n23.R, Exp=ScalarParameter() {defaultValue=1, group=Goo, parameterName=Goo_AOPower})) {name=AO, nodeColor=<rgba 0 1 0.7656 1>}
  NamedRerouteDeclaration(None=Power(Base=@n23.G, Exp=ScalarParameter() {defaultValue=1, group=Goo, parameterName=ThicknessPow})) {name=Thickness, nodeColor=<rgba 0.0547 0 1 1>}
  n24 = Divide(A=NamedRerouteUsage()) {constB=512}
  n25 = Constant()
  n26 = ScalarParameter() {defaultValue=0.175, group=Goo, parameterName=Goo_BounceSpeed}
  n27 = TextureSample(UVs=Panner(
    Coordinate=ComponentMask(None=@n24) {b=true, g=true}
    Speed=MaterialFunctionCall(X=@n25, Y=@n26) {materialFunction=&MakeFloat2}.Result
   ) {bFractionalPart=true, speedY=0.1}) {samplerType=SAMPLERTYPE_LinearColor, texture=/Hermes/Props/BigSplat/BigSplat_LargeSpawner_A/Textures/T_BugSpawner_Goo_TA}
  NamedRerouteDeclaration(None=Multiply(A=@n27.R, B=ScalarParameter() {defaultValue=1, group=Goo, parameterName=Goo_LargeNoiseIntensity})) {name=LargeNoise, nodeColor=<rgba 0.3125 1 0 1>}
  NamedRerouteDeclaration(None=LinearInterpolate(B=NamedRerouteUsage(), Alpha=ScalarParameter() {defaultValue=1, group=Goo, parameterName=Bug_ColorFresnelIntensity}) {constA=1}) {name=SHBiasScale, nodeColor=<rgba 0 0.1328 1 1>}
  n28 = ScalarParameter() {defaultValue=1, parameterName=BasecolorLerp}
  NamedRerouteDeclaration(None=LinearInterpolate(A=@n17, Alpha=LinearInterpolate(A=@n18, Alpha=@n28)) {constB=0}) {name=Basecolor_OUT, nodeColor=<rgba 0.0781 1 0 1>}
  n29 = VectorParameter() {group=Goo, parameterName=Bug_InnerColor, sortPriority=1, defaultValue=<rgba 0.6875 0.1734 0.4093 1>}
  NamedRerouteDeclaration(None=@n29.RGB) {name=SlurpInnerColor, nodeColor=<rgba 0 0.0625 1 1>}
  NamedRerouteDeclaration(None=VectorParameter() {group=Goo, parameterName=Bug_OuterColor, sortPriority=1, defaultValue=<rgba 1 0.3 0.5151 1>}.RGB) {name=SlurpOuterColor, nodeColor=<rgba 1 0 0.8906 1>}
  NamedRerouteDeclaration(None=TextureObject() {samplerType=SAMPLERTYPE_Masks, texture=/Game/Global/MaterialFunctions/TechArt/TidalCrane_Bugs/Textures/T_SlurpBug_TA}) {name=NoiseTex, nodeColor=<rgba 0.1719 1 0 1>}
  n30 = TextureObject() {samplerType=SAMPLERTYPE_LinearColor, texture=&T_BugSpawner_Goo_TA}
  n31 = Constant() {r=25}
  n32 = Constant2Vector() {g=-0.25}
  n33 = Constant()
  n34 = Power(Base=MaterialFunctionCall(Float3=MaterialFunctionCall!(TextureObject=@n30, TextureSize=@n31, WorldPosition, Export Float 4, In=@n32, World Space Normal, ProjectionTransitionContrast=@n33) {materialFunction=/Game/Athena/SafeZone/WorldAlignedTexture_ScalePan}.XYZ Texture) {materialFunction=&BreakOutFloat3Components}.R)
  n35 = Multiply(A=NamedRerouteUsage(), B=Saturate(None=Multiply(A=@n34) {constB=2}))
  n36 = ScalarParameter() {defaultValue=3, group=Goo, parameterName=Goo_InflationWPOAmount}
  NamedRerouteDeclaration(None=Multiply(A=VertexNormalWS(), B=Multiply(A=@n35, B=@n36))) {name=InflationWPO, nodeColor=<rgba 0.875 1 0 1>}
  NamedRerouteDeclaration(None=Multiply(A=@n27.R, B=VectorParameter() {group=Goo, parameterName=Goo_DrippingWPOAmount, defaultValue=<rgba 0 0 -30 1>}.RGB) {constB=5}) {name=BounceWPO, nodeColor=<rgba 0.8047 0 1 1>}
  n37 = LinearInterpolate(
   A=LinearInterpolate(A=NamedRerouteUsage(), B=NamedRerouteUsage(), Alpha=NamedRerouteUsage())
   B=VectorParameter() {group=Goo, parameterName=Caustics_Color, sortPriority=33, defaultValue=<rgba 2 0.1913 0.7472 1>}.RGB
   Alpha=NamedRerouteUsage()
  )
  n38 = LinearInterpolate(B=NamedRerouteUsage(), Alpha=ScalarParameter() {defaultValue=1, group=Goo, parameterName=AO_Intensity}) {constA=1}
  n39 = ScalarParameter() {defaultValue=0.66, group=Goo, parameterName=Bug_TimeofDayScale_Morning}
  n40 = ScalarParameter() {defaultValue=1, group=Goo, parameterName=Bug_TimeofDayScale_Day}
  n41 = ScalarParameter() {defaultValue=0.66, group=Goo, parameterName=Bug_TimeofDayScale_Evening}
  n42 = ScalarParameter() {defaultValue=0.5, group=Goo, parameterName=Bug_TimeofDayScale_Night}
  NamedRerouteDeclaration(None=Multiply(
    A=Multiply(A=@n37, B=@n38)
    B=MaterialFunctionCall(Value_Morning=@n39, Value_Day=@n40, Value_Evening=@n41, Value_Night=@n42) {materialFunction=&Scalar_Time_of_Day}.Result
   )) {name=GooEmissive, nodeColor=<rgba 0.3594 0 1 1>}
  n43 = NamedRerouteUsage()
  n44 = ComponentMask(None=NamedRerouteUsage()) {g=true, r=true}
  n45 = ScalarParameter() {defaultValue=128, group=Goo, parameterName=Caustics_Tiling2, sortPriority=33}
  n46 = VectorParameter() {group=Goo, parameterName=Caustics_Speed2, defaultValue=<rgba -0.0346 0.0324 0 1>}
  n47 = TextureSample(
   UVs=Panner(Coordinate=Divide(A=@n44, B=@n45) {constB=256}, Speed=ComponentMask(None=@n46.RGB) {g=true, r=true}) {bFractionalPart=true}
   Tex=NamedRerouteUsage()
  )
  n48 = Multiply(A=@n47.R) {constB=5}
  n49 = VectorParameter() {group=Goo, parameterName=Caustics_Speed1, defaultValue=<rgba 0.0131 -0.124 0 1>}
  n50 = MaterialFunctionCall(Float3=MaterialFunctionCall(
    TextureObject=NamedRerouteUsage()
    TextureSize=ScalarParameter() {defaultValue=50, group=Goo, parameterName=Caustics_Tiling1, sortPriority=33}
    WorldPosition=Add(A=@n43, B=@n48)
    In=ComponentMask(None=@n49.RGB) {g=true, r=true}
    World Space Normal=NamedRerouteUsage()
   ) {materialFunction=&WorldAlignedTexture_ScalePan}.XYZ Texture) {materialFunction=&BreakOutFloat3Components}
  NamedRerouteDeclaration(None=@n50.G) {name=TP_Particles, nodeColor=<rgba 0 0.2266 1 1>}
  NamedRerouteDeclaration(None=@n50.R) {name=TP_Perlin, nodeColor=<rgba 0.125 1 0 1>}
  n51 = NamedRerouteUsage()
  n52 = Fresnel(Normal=Transform(None=Normalize(VectorInput=@n51))) {baseReflectFraction=0, exponent=1}
  NamedRerouteDeclaration(None=Power(
    Base=FunctionInput(Preview=@n52) {bUsePreviewValueAsDefault=true, inputName=Fresnel, sortPriority=2}
    Exp=ScalarParameter() {defaultValue=1, group=Goo, parameterName=Bug_FresnelPow}
   )) {name=Fresnel, nodeColor=<rgba 0 1 0.0625 1>}
  NamedRerouteDeclaration(None=@n1.Normal) {name=Normal_IN, nodeColor=<rgba 1 0.3516 0 1>}
  NamedRerouteDeclaration() {name=Debug, nodeColor=<rgba 0.2422 1 0 1>}
  n53 = Multiply(
   A=Multiply(A=NamedRerouteDeclaration(None=@n50.B) {name=TP_CausticsNoise, nodeColor=<rgba 0.7109 0 1 1>}, B=@n47.B)
   B=ScalarParameter() {defaultValue=0.1, group=Goo, parameterName=Caustics_Intensity, sortPriority=33}
  )
  n54 = NamedRerouteUsage()
  NamedRerouteDeclaration(None=Multiply(A=Saturate(None=@n53), B=OneMinus(None=@n54))) {name=Caustics, nodeColor=<rgba 1 0.5625 0 1>}
  NamedRerouteDeclaration(None=Power(Base=@n20.R, Exp=ScalarParameter() {defaultValue=3, group=Goo, parameterName=Goo_WPOPower})) {name=VertexColor_Powered, nodeColor=<rgba 0 1 0.7188 1>}
  NamedRerouteUsage()
  Saturate(None=Multiply(A=@n18, B=@n28))
 ]}
