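# Material function for the bug spawner membrane that overrides incoming material attributes with a black-base, emissive-driven look and outputs a Pulsing value.
# Parallax bump-offset veins and noise-animated inner blobs build the color, blending inactive, active and weakpoint states, scaled by time of day.
# Blueprint-fed game times in custom primitive data drive activity progress and curve-atlas spawn portal, bounce and damage blink.
# Breathing and burst motion push vertices along normals through WPO.
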MaterialFunction /Hermes/Props/BigSplat/BigSplat_LargeSpawner_A/Materials/MF_BugSpawner_Membrane
graph {nodes [
  n1 = GetMaterialAttributes(None=FunctionInput() {inputName=Material, inputType=FunctionInput_MaterialAttributes}) {attributeGetTypes [#1, #2]}
  n2 = NamedRerouteUsage()
  n3 = SetMaterialAttributes!(
   MaterialAttributes=@n1.MaterialAttributes
   BaseColor=Constant()
   Metallic=Constant()
   Specular=LinearInterpolate(A=Constant() {r=0.5}, Alpha=@n2) {constB=0}
   Roughness=LinearInterpolate(A=NamedRerouteUsage(), Alpha=@n2)
   EmissiveColor=NamedRerouteUsage()
   WorldPositionOffset=NamedRerouteUsage()
   Normal
  ) {attributeSetTypes [
    #3
    #4
    #5
    #6
    #7
    #2
    #1
   ]}
  FunctionOutput(None=@n3)
  FunctionOutput(None=NamedRerouteUsage()) {outputName=Pulsing}
  n4 = Time()
  LinearInterpolate(Alpha=Frac(None=Multiply(A=@n4) {constB=0.2})) {constB=3}
  n5 = OneMinus(None=NamedRerouteUsage())
  Multiply(A=@n5, B=VertexNormalWS())
  NamedRerouteDeclaration(None=Fresnel(ExponentIn=Constant() {r=1}) {baseReflectFraction=0}) {name=Fresnel, nodeColor=<rgba 1 0.375 0 1>}
  n6 = VectorParameter() {group=Bug Spawner, parameterName=Membrane_OutterColor_Inactive, defaultValue=<rgba 0.1179 0.0574 0.1198 1>}
  n7 = VectorParameter() {group=Bug Spawner, parameterName=Membrane_InnerColor_Inactive, defaultValue=<rgba 0.3438 0.0807 0.3433 1>}
  n8 = NamedRerouteUsage()
  n9 = VectorParameter() {group=Bug Spawner, parameterName=Membrane_OutterColor_Active, defaultValue=<rgba 0.4583 0.1555 0.4105 1>}
  n10 = VectorParameter() {group=Bug Spawner, parameterName=Membrane_InnerColor_Active, defaultValue=<rgba 1.9744 0.6563 2 1>}
  n11 = LinearInterpolate(
   A=LinearInterpolate(A=@n6.RGB, B=@n7.RGB, Alpha=@n8)
   B=LinearInterpolate(
    A=LinearInterpolate(A=@n9.RGB, B=@n10.RGB, Alpha=@n8)
    B=VectorParameter() {group=Bug Spawner, parameterName=Membrane_WeakpointColor, defaultValue=<rgba 2 1.4009 0.4271 1>}.RGB
    Alpha=NamedRerouteUsage()
   )
   Alpha=NamedRerouteUsage()
  )
  n12 = VectorParameter() {group=Bug Spawner, parameterName=Membrane_VeinColor_Inner, defaultValue=<rgba 0.1784 0.0399 0.1823 1>}
  n13 = LinearInterpolate(
   A=@n11
   B=LinearInterpolate(A=Multiply(A=@n12.RGB) {constB=0.5}, B=@n12.RGB, Alpha=NamedRerouteUsage()) {constA=0.5}
   Alpha=NamedRerouteUsage()
  )
  n14 = VectorParameter() {group=Bug Spawner, parameterName=Membrane_VeinColor_Outter, defaultValue=<rgba 0.0766 0.0222 0.0781 1>}
  n15 = LinearInterpolate(
   A=@n14.RGB
   B=Multiply(A=@n14.RGB, B=ScalarParameter() {defaultValue=3, group=Bug Spawner, parameterName=Membrane_VeinColor_ActiveMulti}) {constB=3}
   Alpha=NamedRerouteUsage()
  )
  n16 = NamedRerouteUsage()
  n17 = NamedRerouteUsage()
  n18 = ScalarParameter() {defaultValue=0.2, group=Bug Spawner, parameterName=BrightnessRandom}
  n19 = Multiply(
   A=Multiply(A=LinearInterpolate(A=@n13, B=@n15, Alpha=@n16) {desc=Color with veins}, B=LinearInterpolate(B=@n17, Alpha=@n18) {constA=1, constAlpha=0.2})
   B=LinearInterpolate(Alpha=NamedRerouteUsage()) {constA=1, constB=10}
  )
  n20 = Multiply(A=@n11, B=VectorParameter() {group=Bug Spawner, parameterName=PortalColor, defaultValue=<rgba 0.5 0.2188 0.4898 1>}.RGB)
  n21 = NamedRerouteUsage()
  n22 = Multiply(
   A=LinearInterpolate(
    A=LinearInterpolate(A=@n19, B=@n20, Alpha=@n21)
    B=VectorParameter() {group=Bug Spawner, parameterName=PortalEdgeColor, defaultValue=<rgba 2.8972 0.875 3 1>}.RGB
    Alpha=NamedRerouteUsage()
   )
   B=LinearInterpolate(Alpha=NamedRerouteUsage()) {constA=0.75, constB=1.5}
  )
  n23 = LinearInterpolate(Alpha=Multiply(A=NamedRerouteUsage()) {constB=0.5}) {constA=1, constB=3}
  n24 = ScalarParameter() {defaultValue=0.66, group=Bug, parameterName=Bug_TimeofDayScale_Morning}
  n25 = ScalarParameter() {defaultValue=1, group=Bug, parameterName=Bug_TimeofDayScale_Day}
  n26 = ScalarParameter() {defaultValue=0.66, group=Bug, parameterName=Bug_TimeofDayScale_Evening}
  n27 = ScalarParameter() {defaultValue=0.5, group=Bug, parameterName=Bug_TimeofDayScale_Night}
  NamedRerouteDeclaration(None=Multiply(
    A=Multiply(A=@n22, B=@n23)
    B=MaterialFunctionCall(Value_Morning=@n24, Value_Day=@n25, Value_Evening=@n26, Value_Night=@n27) {materialFunction=&Scalar_Time_of_Day}.Result
   )) {name=Emissive_OUT, nodeColor=<rgba 0.2422 1 0 1>}
  NamedRerouteDeclaration(None=@n1.WorldPositionOffset) {name=WPO_IN, nodeColor=<rgba 0.9688 0 1 1>}
  n28 = VectorParameter() {
   parameterName=DEBUG_Activity
   channelNames {a=GameTimeFromBP_Damage, b=GameTimeFromBP_Spawning, g=IsActivating, r=GameTimeFromBP_Activity}
   defaultValue=<rgba 0 1 0 1>
  }
  n29 = VectorParameter() {
   bUseCustomPrimitiveData=true
   parameterName=Activity
   channelNames {a=GameTimeFromBP_Damage, b=GameTimeFromBP_Spawning, g=IsActivating, r=GameTimeFromBP_Activity}
   defaultValue=<rgba 0 0 0 1>
  }
  n30 = MaterialFunctionCall(Float4=StaticSwitchParameter(
    True=MaterialFunctionCall(X=@n28.GameTimeFromBP_Activity, Y=@n28.GameTimeFromBP_Activity, Z=@n28.GameTimeFromBP_Activity, A=@n28.GameTimeFromBP_Activity) {materialFunction=&MakeFloat4}.Result
    False=MaterialFunctionCall(X=@n29.GameTimeFromBP_Activity, Y=@n29.GameTimeFromBP_Activity, Z=@n29.GameTimeFromBP_Activity, A=@n29.GameTimeFromBP_Activity) {materialFunction=&MakeFloat4}.Result
   ) {parameterName=DEBUG}) {materialFunction=&BreakOutFloat4Components}
  NamedRerouteDeclaration(None=@n30.B) {name=GameTime_Spawning, nodeColor=<rgba 1 0.8203 0 1>}
  NamedRerouteDeclaration(None=@n30.R) {name=GameTime_Activity, nodeColor=<rgba 0.5 0 1 1>}
  NamedRerouteDeclaration(None=@n30.G) {name=IsActivating, nodeColor=<rgba 0 1 0.1797 1>}
  NamedRerouteDeclaration(None=@n30.A) {name=GameTime_Damage, nodeColor=<rgba 1 0 0.1406 1>}
  n31 = CurveAtlasRowParameter(CurveTime=NamedRerouteUsage()) {
   atlas=/Hermes/Props/BigSplat/BigSplat_LargeSpawner_A/Textures/CA_BugSpawner
   curve=/Hermes/Props/BigSplat/BigSplat_LargeSpawner_A/Textures/C_BugSpawner
   group=Bug Spawner
   parameterName=Spawn
  }
  NamedRerouteDeclaration(None=@n31.G) {name=SpawningPortalProgress, nodeColor=<rgba 0.8516 0 1 1>}
  n32 = NamedRerouteUsage()
  n33 = OneMinus(None=NamedRerouteUsage())
  n34 = TextureSample(UVs=NamedRerouteUsage()) {desc=TO-DO: USE DF TEX, samplerSource=SSM_Wrap_WorldGroupSettings, texture=/Game/Global/Textures/Default/Tileable_Patterns/T_Veins_Blurred}
  n35 = Subtract(A=Multiply(A=Add(A=@n32, B=@n33)) {constB=2}, B=Multiply(A=OneMinus(None=@n34.R)))
  n36 = OneMinus(None=SmoothStep(Value=@n35) {constMax=0.6, constMin=0.5})
  NamedRerouteDeclaration(None=@n36) {name=PortalMask, nodeColor=<rgba 1 0 0.8438 1>}
  n37 = Constant() {r=0.5}
  n38 = SmoothStep(Value=OneMinus(None=Distance(A=@n35, B=@n37))) {constMin=0.7}
  NamedRerouteDeclaration(None=Saturate(None=Subtract(A=@n38, B=@n36))) {name=PortalEdge, nodeColor=<rgba 0 1 0.8359 1>}
  n39 = TextureSample(UVs=TextureCoordinate(), Tex=NamedRerouteUsage()) {
   samplerSource=SSM_Wrap_WorldGroupSettings
   samplerType=SAMPLERTYPE_Masks
   texture=/Hermes/Props/BigSplat/BigSplat_LargeSpawner_A/Textures/T_BugSpawner_Membrane_Webs
  }
  n40 = NamedRerouteDeclaration(None=@n39.B) {name=SphereMask, nodeColor=<rgba 0 1 0.25 1>}
  n41 = NamedRerouteUsage()
  n42 = NamedRerouteUsage()
  n43 = ScalarParameter() {defaultValue=1.5, group=Bug Spawner, parameterName=Webs_Depth}
  n44 = Constant()
  n45 = Constant() {r=0.1}
  n46 = NamedRerouteUsage()
  n47 = TextureSample(
   UVs=MaterialFunctionCall!(UV Coord=@n41, Height=Multiply(A=@n42, B=@n43), ReferencePlane=@n44, HeightRatio=@n45, Normal) {materialFunction=/Game/Global/MaterialFunctions/MF_BumpOffset_Accurate}.Result
   Tex=@n46
  ) {
   samplerSource=SSM_Wrap_WorldGroupSettings
   samplerType=SAMPLERTYPE_Masks
   texture=&T_BugSpawner_Membrane_Webs
  }
  n48 = Multiply(A=@n47.R, B=ScalarParameter() {defaultValue=0.6, parameterName=VeinsOpacity})
  NamedRerouteDeclaration(None=SmoothStep(Value=Add(A=@n40, B=@n48)) {constMin=0.2}) {name=Veins, nodeColor=<rgba 0 1 0.1562 1>}
  n49 = Multiply(A=NamedRerouteUsage()) {constB=0.5}
  n50 = TextureSample(UVs=Add(A=Panner(Coordinate=@n49) {speedX=0.123, speedY=-0.0993}, B=NamedRerouteUsage()) {desc=TO-DO}) {samplerSource=SSM_Wrap_WorldGroupSettings, texture=/Game/Global/Textures/Default/Noise/3xPerlin_M}
  NamedRerouteDeclaration(None=@n50.R) {name=PerlinNoiseR, nodeColor=<rgba 0.4766 1 0 1>}
  n51 = Multiply(A=TextureCoordinate(), B=ScalarParameter() {defaultValue=0.75, group=Bug Pod, parameterName=SkinSpots_Tiling})
  n52 = TextureSample(UVs=@n51) {
   desc=Waiting for HS to add Normal details from texture
   samplerType=SAMPLERTYPE_Normal
   texture=/Hermes/Props/BigSplat/BigSplat_LargeSpawner_A/Textures/T_BugSpawner_Membrane_N
  }
  n53 = MaterialFunctionCall(Normal=@n52.RGB, Flatness=ScalarParameter() {defaultValue=0.5, group=Bug Pod, parameterName=SkinSpots_NormalFlatness}) {materialFunction=&FlattenNormal}
  n54 = Constant3Vector() {constant=<rgba 0 0 1 0>}
  n55 = NamedRerouteUsage()
  NamedRerouteDeclaration(None=MaterialFunctionCall(BaseNormal=NamedRerouteUsage(), AdditionalNormal=LinearInterpolate(A=@n53.Result, B=@n54, Alpha=@n55)) {materialFunction=&BlendAngleCorrectedNormals}.Result) {name=Normal_OUT, nodeColor=<rgba 0.4531 0 1 1>}
  n56 = TextureSample(UVs=@n51) {
   samplerSource=SSM_Wrap_WorldGroupSettings
   samplerType=SAMPLERTYPE_Masks
   texture=/Hermes/Props/BigSplat/BigSplat_LargeSpawner_A/Textures/T_BugSpawner_Membrane_TA
  }
  n57 = NamedRerouteDeclaration(None=@n56.B) {name=SkinSpots, nodeColor=<rgba 0 0.2031 1 1>}
  NamedRerouteDeclaration(None=@n57) {name=Roughness_OUT, nodeColor=<rgba 0 0.2969 1 1>}
  NamedRerouteDeclaration(None=LinearInterpolate(Alpha=@n57) {constA=0.5}) {name=Specular_OUT, nodeColor=<rgba 0.1719 1 0 1>}
  NamedRerouteDeclaration(None=@n1.Normal) {name=Normal_IN, nodeColor=<rgba 0.125 1 0 1>}
  n58 = MaterialFunctionCall(
   UV Coord=MaterialFunctionCall(UVs=NamedRerouteUsage(), Texture Scale=ScalarParameter() {defaultValue=30, group=Bug Spawner, parameterName=Membrane_BlobSize}) {materialFunction=&ScaleUVsByCenter}.UVs
   Height=ScalarParameter() {defaultValue=3, group=Bug Spawner, parameterName=Membrane_BlobHeight}
   ReferencePlane=Constant()
   HeightRatio=Constant() {r=0.1}
  ) {materialFunction=&MF_BumpOffset_Accurate}
  n59 = Power(Base=SphereMask(A=@n58.Result, B=Constant() {r=0.5}, Radius=ScalarParameter() {defaultValue=0.6, group=Bug Spawner, parameterName=Membrane_BlobRadius}) {attenuationRadius=0.3, hardnessPercent=0}) {constExponent=3}
  MaterialExpressionNamedRerouteDeclaration_25 = NamedRerouteDeclaration(None=@n59) {name=InnerSphereMask, nodeColor=<rgba 0 0.9297 1 1>}
  n60 = Multiply(A=@n58.Result) {constB=0.5}
  n61 = Add(A=Multiply(A=Time(), B=ScalarParameter() {defaultValue=0.5, group=Bug Spawner, parameterName=Membrane_MoveSpeed}), B=NamedRerouteUsage())
  MaterialExpressionMultiply_13 = Multiply(
   A=TextureSample(UVs=Panner(Coordinate=@n60, Time=@n61) {bFractionalPart=true, speedX=0.0984, speedY=-0.111}) {samplerSource=SSM_Wrap_WorldGroupSettings, texture=&3xPerlin_M}.R
   B=TextureSample(UVs=Panner(Coordinate=@n60, Time=@n61) {bFractionalPart=true, speedX=-0.021, speedY=0.23}) {samplerSource=SSM_Wrap_WorldGroupSettings, texture=&3xPerlin_M}.G
  ) {desc=TO-DO: SIMPLIFY}
  MaterialExpressionConstant_12 = Constant() {r=0.01}
  n62 = Custom!(a=@MaterialExpressionNamedRerouteDeclaration_25, b=@MaterialExpressionMultiply_13, k=@MaterialExpressionConstant_12) {
   code="float h = max( k-abs(a-b), 0.0 )/k;\r\nreturn min( a, b ) - h*h*k*(1.0/4.0);"
   desc=SmoothMin
   description=SmoothMin
   outputType=CMOT_Float1
   inputs [
    {
     inputName=a
     input {expression=&self:MaterialExpressionNamedRerouteDeclaration_25}
    }
    {
     inputName=b
     input {expression=&self:MaterialExpressionMultiply_13}
    }
    {
     inputName=k
     input {expression=&self:MaterialExpressionConstant_12}
    }
   ]
  }
  n63 = Multiply(
   A=Add(
    A=Multiply(A=@MaterialExpressionNamedRerouteDeclaration_25) {constB=0.1}
    B=Multiply(A=@MaterialExpressionNamedRerouteDeclaration_25, B=@n62) {constB=0.1}
   ) {desc=TO-DO: BETTER BUG MOVEMENT}
   B=ScalarParameter() {defaultValue=10, group=Bug Spawner, parameterName=Membrane_InnerBrightness}
  ) {constB=4}
  NamedRerouteDeclaration(None=Saturate(None=SmoothStep(Value=@n63))) {name=BlobMask, nodeColor=<rgba 1 0.375 0 1>}
  NamedRerouteDeclaration(None=SmoothStep(Value=@n59) {constMin=0.6}) {name=WeakpointMask, nodeColor=<rgba 1 0.9375 0 1>}
  NamedRerouteDeclaration() {name=Debug, nodeColor=<rgba 1 0 0.3516 1>}
  NamedRerouteDeclaration(None=@n50.B) {name=PerlinNoiseB, nodeColor=<rgba 0 0.8828 1 1>}
  n64 = Add(
   A=Multiply(A=Time(), B=ScalarParameter() {defaultValue=0.6, group=Bug Spawner, parameterName=Membrane_PulseSpeed})
   B=Multiply(A=NamedRerouteUsage(), B=ScalarParameter() {defaultValue=2, group=Bug Spawner, parameterName=Pulse_VertColorInflence}) {constB=2}
  )
  n65 = Multiply(A=ConstantBiasScale(None=Sine(None=@n64)), B=NamedRerouteUsage())
  NamedRerouteDeclaration(None=Multiply(
    A=NamedRerouteDeclaration(None=@n65) {name=Pulsing, nodeColor=<rgba 0 1 0.5547 1>}
    B=ScalarParameter() {defaultValue=10, group=Bug Spawner, parameterName=Membrane_Breathing_WPOAmount}
   )) {name=BreathingWPO, nodeColor=<rgba 0.9688 1 0 1>}
  n66 = TextureSample(
   UVs=MaterialFunctionCall(
    UV Coord=@n41
    Height=ScalarParameter() {defaultValue=2, group=Bug Spawner, parameterName=Webs_Depth_Inner}
    ReferencePlane=@n44
    HeightRatio=@n45
   ) {materialFunction=&MF_BumpOffset_Accurate}.Result
   Tex=@n46
  ) {
   samplerSource=SSM_Wrap_WorldGroupSettings
   samplerType=SAMPLERTYPE_Masks
   texture=&T_BugSpawner_Membrane_Webs
  }
  NamedRerouteDeclaration(None=@n66.G) {name=Veins2, nodeColor=<rgba 0 1 0.1562 1>}
  NamedRerouteDeclaration(None=TextureCoordinate() {coordinateIndex=1}) {name=UVs, nodeColor=<rgba 0.4766 0 1 1>}
  NamedRerouteDeclaration(None=StaticSwitchParameter(True=ScalarParameter() {parameterName=DEBUG_SpawnProgress, sliderMax=1}, False=@n4) {parameterName=DEBUG}) {name=Time, nodeColor=<rgba 0 1 0.7422 1>}
  n67 = TextureObjectParameter() {
   group=Bug Spawner
   parameterName=MembraneTexture
   samplerType=SAMPLERTYPE_Masks
   texture=&T_BugSpawner_Membrane_Webs
  }
  NamedRerouteDeclaration(None=@n67) {name=WebsTex, nodeColor=<rgba 0 0.5547 1 1>}
  NamedRerouteDeclaration(None=ScalarParameter() {bUseCustomPrimitiveData=true, group=Bug Spawner, parameterName=GameTimeFromBP_Anticipation, primitiveDataIndex=4}) {name=GameTime_Anticipation, nodeColor=<rgba 1 0 0.1406 1>}
  n68 = Multiply(A=@n5, B=Add(A=NamedRerouteUsage(), B=NamedRerouteUsage()))
  n69 = VertexNormalWS()
  NamedRerouteDeclaration(None=Add(A=NamedRerouteUsage(), B=Multiply(A=@n68, B=@n69))) {name=WPO_OUT, nodeColor=<rgba 1 0 0.8203 1>}
  n70 = Subtract(A=Time(), B=NamedRerouteUsage())
  n71 = ScalarParameter() {defaultValue=1.2, group=Bug Spawner, parameterName=DamageDuration}
  n72 = CurveAtlasRowParameter(CurveTime=Saturate(None=Divide(A=@n70, B=@n71))) {
   atlas=&CA_BugSpawner
   curve=&C_BugSpawner
   group=Bug Spawner
   parameterName=Spawn
  }
  NamedRerouteDeclaration(None=@n72.B) {name=DamageBlink, nodeColor=<rgba 0 0.7188 1 1>}
  n73 = Subtract(A=NamedRerouteUsage(), B=NamedRerouteUsage())
  n74 = ScalarParameter() {defaultValue=1, group=Bug Spawner, parameterName=BurstDuration}
  NamedRerouteDeclaration(None=Saturate(None=Divide(A=@n73, B=@n74))) {name=BurstProgress, nodeColor=<rgba 0 0.7188 1 1>}
  NamedRerouteDeclaration(None=ChannelMaskParameter(None=VertexColor()) {group=Bug Spawner, parameterName=RandomChannel}.R) {name=VC_Random, nodeColor=<rgba 1 0 0.8203 1>}
  n75 = ConstantBiasScale(None=@n31.R) {bias=-0.5, scale=2}
  NamedRerouteDeclaration(None=Multiply(
    A=NamedRerouteDeclaration(None=@n75) {name=SpawningBounce, nodeColor=<rgba 1 0 0.5859 1>}
    B=ScalarParameter() {defaultValue=40, group=Bug Spawner, parameterName=Membrane_Burst_WPOAmount}
   )) {name=BurstWPO, nodeColor=<rgba 1 0.1641 0 1>}
  n76 = Time()
  n77 = NamedRerouteUsage()
  n78 = Saturate(None=Divide(A=Subtract(A=@n76, B=@n77), B=ScalarParameter() {defaultValue=1.5, group=Bug Spawner, parameterName=ActivityDuration}))
  NamedRerouteDeclaration(None=LinearInterpolate(A=OneMinus(None=@n78), B=@n78, Alpha=NamedRerouteUsage())) {name=ActivityProgress, nodeColor=<rgba 0.5469 1 0 1>}
  NamedRerouteUsage()
  NamedRerouteUsage()
  ScalarParameter() {defaultValue=0.5, parameterName=DEBUG_ActivityProgress}
  TextureSample() {texture=&T_Veins_Blurred}
 ]}
