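# Parent material for the BigSplat large bug spawner, shading its membrane goo or, via a static switch, a goo pool.
# Tiled, flattenable detail normals and goo surface parameters feed the membrane function, whose pulsing also drives the goo pool variant.
# World position offset blends out by distance, then tech-art override, destructible damage, Curie fallback and Creative move tool functions layer on.
# Low-quality mobile swaps in constant specular and roughness.
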Material /Hermes/Props/BigSplat/BigSplat_LargeSpawner_A/Materials/M_BugSpawner_Parent
graph {
 properties {
  bAutomaticallySetUsageInEditor=false
  bUseMaterialAttributes=true
  bUsedWithMeshDeformer=true
  bUsedWithNanite=true
  bUsedWithSkeletalMesh=true
  shadingModel=MSM_FromMaterialExpression
  usedShadingModels=Default Lit
 }
 nodes [
  n1 = TextureSampleParameter2D(UVs=Multiply(
    A=TextureCoordinate() {coordinateIndex=1}
    B=ScalarParameter() {defaultValue=0.75, group=Bug Spawner, parameterName=Spawner_NormalDetails_Tiling}
   )) {
   group=00_Texture Inputs
   parameterName=Normals
   samplerType=SAMPLERTYPE_Normal
   sortPriority=2
   texture=&FlatNormalBlue
  }
  n2 = ScalarParameter() {defaultValue=0.5, parameterName=Spawner_NormalDetails_Flatness}
  n3 = SetMaterialAttributes!(
   MaterialAttributes
   Normal=StaticSwitchParameter(
    True=Constant3Vector() {constant=<rgba 0 0 1 0>}
    False=MaterialFunctionCall(Normal=@n1.RGB, Flatness=@n2) {materialFunction=&FlattenNormal}.Result
   ) {group=Bug Spawner, parameterName=UseFlatNormal}
   Base Color=Constant()
   Specular=ScalarParameter() {defaultValue=0.5, group=Bug Spawner, parameterName=MembraneGoo_Specular}
   Roughness=ScalarParameter() {defaultValue=0.18, group=Bug Spawner, parameterName=MembraneGoo_Roughness}
   Metallic=Constant()
   Emissive Color
   Shading Model=ShadingModel()
  ) {attributeSetTypes [
    #1
    #2
    #3
    #4
    #5
    #6
    #7
   ]}
  n4 = MaterialFunctionCall!(Material=@n3) {materialFunction=&/MF_BugSpawner_Membrane}
  n5 = GetMaterialAttributes(None=StaticSwitchParameter(
    True=MaterialFunctionCall!(Material=@n3, Pulsing=@n4.Pulsing, Fresnel) {materialFunction=&/MF_BugSpawner_GooPool}.Result
    False=@n4.Result
   ) {group=Bug Spawner, parameterName=UseGooPool, sortPriority=0}) {attributeGetTypes [#8]}
  n6 = SetMaterialAttributes!(
   MaterialAttributes=@n5.MaterialAttributes
   World Position Offset=LinearInterpolate(A=@n5.World Position Offset, Alpha=NamedRerouteUsage()) {constB=0}
   Emissive Color
  ) {attributeSetTypes [#8, #6]}
  n7 = MaterialFunctionCall!(
   In=MaterialFunctionCall(
    Material_In=MaterialFunctionCall(In=@n6) {materialFunction=&MF_TechArtOverrides}.Result
    UseAdditiveWPO=StaticBool() {value=true}
   ) {materialFunction=&MF_FN_DestructibleDamage}.Material_Out
   UseSphericalMappingOverride
  ) {materialFunction=/Game/Packages/Fortress_SharedMaterials/Base_Material_Functions/MF_CurieFallbackEffect}
  Composite!(InputPin=@n7.Result) {subgraphName=Mobile Opt}
  MaterialExpressionReroute_0 = Reroute(None=@n7.Result)
  n8 = GetMaterialAttributes(None=@MaterialExpressionReroute_0.InputPin)
  n9 = Constant() {r=0.5}
  n10 = Constant() {r=0.8}
  n11 = ShadingPathSwitch(
   Default=@MaterialExpressionReroute_0.InputPin
   Mobile=QualitySwitch(
    Default=@MaterialExpressionReroute_0.InputPin
    Low=SetMaterialAttributes!(MaterialAttributes=@n8.MaterialAttributes, Specular=@n9, Roughness=@n10) {attributeSetTypes [#3, #4]}
    High=@MaterialExpressionReroute_0.InputPin
    Medium=@MaterialExpressionReroute_0.InputPin
    Epic=@MaterialExpressionReroute_0.InputPin
   )
  )
  MaterialExpressionReroute_1 = Reroute(None=@n11)
  MaterialFunctionCall(MA In=@MaterialExpressionReroute_1.Output) {materialFunction=&MF_CreativeMoveTool}
  NamedRerouteDeclaration(None=MaterialFunctionCall!(FadeBlendDistance=ScalarParameter() {defaultValue=3000, parameterName=Fade Blend Distance}, AutoHandleBadDistance) {materialFunction=/Game/Global/MaterialFunctions/TechArt/MF_WPODisableBlend}.VertexShaderOutput) {name=Disable Blend, nodeColor=<rgba 0 0.7891 1 1>}
  PinBase!(Output=@n11) {reroutePins [Output {expression=&self:MaterialExpressionReroute_1}]}
  PinBase!() {reroutePins [InputPin {expression=&self:MaterialExpressionReroute_0}]}
 ]
}
